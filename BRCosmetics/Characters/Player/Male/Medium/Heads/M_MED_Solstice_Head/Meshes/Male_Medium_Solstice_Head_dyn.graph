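# Physics asset for the Solstice male medium head, providing bodies and constraints for its three dynamic hair chains.
# Each chain pairs a small capsule at its root bone with a larger offset capsule on its child, joined by tight swing and twist limits.
# Each root is constrained to a large spine_05 capsule representing the upper body, with swing and twist left free.
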PhysicsAsset /BRCosmetics/Characters/Player/Male/Medium/Heads/M_MED_Solstice_Head/Meshes/Male_Medium_Solstice_Head_dyn
bodies {
 dyn_hair_01 [{shapeType=Capsule, length=0.5, radius=0.5, shapeName=dyn_hair_01_capsule}]
 dyn_hair_01_a [{shapeType=Capsule, length=8.2373, radius=4.165, shapeName=dyn_hair_01_capsule, center=<3.8575 0 3.8>, rotation=<rot 90 0 0>}]
 dyn_hair_02 [{shapeType=Capsule, length=0.5, radius=0.5, shapeName=dyn_hair_02_capsule}]
 dyn_hair_02_a [{shapeType=Capsule, length=7.0094, radius=5.3929, shapeName=dyn_hair_01_capsule, center=<3.8575 0 3.8>, rotation=<rot 90 0 0>}]
 dyn_hair_03 [{shapeType=Capsule, length=0.5, radius=0.5, shapeName=dyn_hair_03_capsule}]
 dyn_hair_03_a [{shapeType=Capsule, length=8.2373, radius=4.165, shapeName=dyn_hair_01_capsule, center=<3.8575 0 3.8>, rotation=<rot 90 0 0>}]
 spine_05 [{shapeType=Capsule, length=15.7867, radius=24.1374, shapeName=spine_05_capsule, center=<-5.8386 -7.6009 0>}]
}
constraints [
 {bone1Name=dyn_hair_01_a, bone2Name=dyn_hair_01, swing1LimitDegrees=15, swing2LimitDegrees=15, twistLimitDegrees=5}
 {bone1Name=dyn_hair_02_a, bone2Name=dyn_hair_02, swing1LimitDegrees=15, swing2LimitDegrees=15, twistLimitDegrees=5}
 {bone1Name=dyn_hair_03_a, bone2Name=dyn_hair_03, swing1LimitDegrees=15, swing2LimitDegrees=15, twistLimitDegrees=5}
 {
  bone1Name=dyn_hair_01
  bone2Name=spine_05
  swing1LimitDegrees=45
  swing1Motion=Free
  swing2LimitDegrees=45
  swing2Motion=Free
  twistLimitDegrees=45
  twistMotion=Free
 }
 {
  bone1Name=dyn_hair_02
  bone2Name=spine_05
  swing1LimitDegrees=45
  swing1Motion=Free
  swing2LimitDegrees=45
  swing2Motion=Free
  twistLimitDegrees=45
  twistMotion=Free
 }
 {
  bone1Name=dyn_hair_03
  bone2Name=spine_05
  swing1LimitDegrees=45
  swing1Motion=Free
  swing2LimitDegrees=45
  swing2Motion=Free
  twistLimitDegrees=45
  twistMotion=Free
 }
]
